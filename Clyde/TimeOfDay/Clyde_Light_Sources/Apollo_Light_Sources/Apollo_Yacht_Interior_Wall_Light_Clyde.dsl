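; Yacht interior wall light actor that removes itself when a structure it is attached to is destroyed.
; On BeginPlay it loops over its Light Attachment Actors array and binds each valid entry's OnDied event to a DestroyedWall handler.
; When that event fires with a valid damaged actor, the whole light actor is destroyed.
; The construction script destroys the point light component when the killLight flag is set.
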(event EventBeginPlay
  (if (> (Utilities|Array|Length (Variables|Default|GetLight%20Attachment%20Actors)) 0)
    (for _array_element (Variables|Default|GetLight%20Attachment%20Actors)
      (Utilities|IsValid _array_element
        (:Is%20Valid
          (EventDispatchers|BindEventtoOnDied _array_element (delegate Custom|DestroyedWall)))))))

(event Custom|DestroyedWall
  (params (DamagedActor :type Actor%20Object%20Reference) (Damage :type Float%20%28single-precision%29) (InstigatedBy :type Controller%20Object%20Reference) (DamageCauser :type Actor%20Object%20Reference) (HitLocation :type Vector) (FHitComponent :type Primitive%20Component%20Object%20Reference) (BoneName :type Name) (Momentum :type Vector))
  (Utilities|IsValid DamagedActor
    (:Is%20Valid
      (Actor|DestroyActor self))))

(fn ConstructionScript
  (if (Variables|Lighting|GetkillLight)
    (Components|DestroyComponent (Variables|Default|GetPointLight))))
